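# Niagara system for the ClumsyChew emote, drawing a warm light beam effect with lens flare and drifting dust over about 6.6 seconds.
# Light_Cones bursts tagged spawn groups into mesh cones and glow/ring sprites, with curve-driven startup pop, flicker and time-of-day glow colors.
# Front cones and the pulsing Rainbow_Flare fade by camera angle, while Dust drifts with curl noise and gravity.
# Light_Meshes, Light_Meshes001 and Tertiary_Details are disabled.
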NiagaraSystem /BRCosmetics/Animation/Game/MainPlayer/Emotes/ClumsyChew/FX/NS_Emote_ClumsyChew
dependencies {
 usedDataInterfaces [&NiagaraDataInterfaceCurve, &NiagaraDataInterfaceCamera, &NiagaraDataInterfaceVectorCurve]
 usedDynamicInputs [&FNTimeOfDayColor]
 usedModules [
  &self:SystemSpawnScript.SystemScriptSource.SystemScriptGraph.NiagaraNodeAssignment_0.SetVariables_#1
  &self:SystemSpawnScript.SystemScriptSource.SystemScriptGraph.NiagaraNodeAssignment_1.SetVariables_#2
  &SystemState
 ]
}
emitters {
 Dust {
  emitterUpdateScript [
   &EmitterState(Life Cycle Mode=::NewEnumerator1, Loop Behavior=::NewEnumerator1, Loop Duration=6.55, UseLoopDelay=false)
   &SpawnRate(
    SpawnRate=&RandomRangeFloat(Minimum=0.4, Maximum=3, Evaluation Type=::NewEnumerator1)
    Use Spawn Probability=false
   )
   &SpawnBurst_Instantaneous(Spawn Count=3, Use Spawn Probability=false, UseLoopCountLimit=false)
  ]
  particleSpawnScript [
   &InitializeParticle(
    Color=<rgba 0.8781 0.6468 0.486 0.25>
    Lifetime Max=1.5
    Lifetime Min=0.8
    Lifetime Mode=::NewEnumerator1
    Position Offset=<35 -13 -7>
    Sprite Size Mode=::NewEnumerator1
    Uniform Sprite Size Max=2
    Uniform Sprite Size Min=0.25
    UsePositionOffset=true
    Write Material Random=true
   )
   &ShapeLocation(
    Box Size=<20 20 20>
    MaskSpawnGroup=false
    Override Seed=false
    Rotation Angle Type:ENiagara_AngleInput=::NewEnumerator2
    Rotation Mode=::NewEnumerator2
    Shape Primitive=::NewEnumerator2
    Surface Only=false
    UseOwnerScale=true
    Yaw / Pitch / Roll=<0 0.5 0>
   )
   &CurlNoiseForce(Bias Noise=false, Noise Frequency=60, Noise Strength=8, Pan Noise=false)
   &ApplyInitialForces()
  ]
  particleUpdateScript [
   &ParticleState()
   &ScaleSpriteSize(Uniform Curve Sprite Scale=DataInterface() {propertyValues {Curve {keys [
        {arriveTangent=0.0255, interpMode=RCIM_Cubic, leaveTangent=0.0255, tangentMode=RCTM_User, time=0.1564, value=0.0471}
        {interpMode=RCIM_Cubic, tangentMode=RCTM_User, time=0.4614, value=1}
        {interpMode=RCIM_Cubic, tangentMode=RCTM_User, time=0.6827, value=1}
        {arriveTangent=-0.0681, interpMode=RCIM_Cubic, leaveTangent=-0.0681, tangentMode=RCTM_User, time=1, value=0}
       ]}}})
   /SetVariables(Particles.Position=&AddVectorToPosition(
     Position=$Particles.Position
     Vector=&MakeVector(
      X=&FloatFromCurve(
       FloatCurve=DataInterface() {propertyValues {Curve {keys [<at 0 0>,<at 1 1>]}}}
       Scale Curve=&RandomRangeFloat(Minimum=1, Maximum=2)
      )
      Y=0
      Z=0
     )
    )) {isSetParameters=true}
   &Drag(Use Linear Drag=true, Use Rotational Drag=true)
   &GravityForce(Gravity=&RandomRangeVector(Minimum=<0 0 -50>, Maximum=<0 0 -100>))
   &CurlNoiseForce(Bias Noise=false, Noise Frequency=60, Noise Strength=8, Pan Noise=false) {module=CurlNoiseForce001}
   &SolveForcesAndVelocity(Clamp Velocity=false, Limit Acceleration=false)
   &ScaleColor(
    Scale Alpha=&FloatFromCurve(FloatCurve=DataInterface() {propertyValues {Curve {keys [
         {arriveTangent=0.0255, interpMode=RCIM_Cubic, leaveTangent=0.0255, tangentMode=RCTM_User, time=0.1564, value=0.0471}
         {interpMode=RCIM_Cubic, tangentMode=RCTM_User, time=0.3214, value=0.9173}
         {arriveTangent=-1.3518, interpMode=RCIM_Cubic, leaveTangent=-1.3518, time=0.6587, value=0.244}
         {arriveTangent=-0.0681, interpMode=RCIM_Cubic, leaveTangent=-0.0681, tangentMode=RCTM_User, time=1, value=0}
        ]}}})
    ScaleA=true
    ScaleRGB=true
   )
  ]
  rendererClasses [&NiagaraSpriteRendererProperties]
  renderers [{rendererClass=&NiagaraSpriteRendererProperties}]
 }
 Faint_Glow {rendererClasses [], renderers []}
 Flare {rendererClasses [], renderers []}
 Glow {rendererClasses [], renderers []}
 Hot_Ring {rendererClasses [], renderers []}
 Light {rendererClasses [], renderers []}
 Light_Cones {
  emitterUpdateScript [
   &EmitterState(Life Cycle Mode=::NewEnumerator1, Loop Behavior=::NewEnumerator1, Loop Duration=6.616, UseLoopDelay=false)
   &SpawnBurst_Instantaneous(Use Spawn Probability=false, UseLoopCountLimit=false)
   &SpawnBurst_Instantaneous(Spawn Group=1, Use Spawn Probability=false, UseLoopCountLimit=false) {module=SpawnBurst_Instantaneous001}
   &SpawnBurst_Instantaneous(Spawn Group=2, Use Spawn Probability=false, UseLoopCountLimit=false) {module=SpawnBurst_Instantaneous002}
   &SpawnBurst_Instantaneous(Spawn Group=3, Use Spawn Probability=false, UseLoopCountLimit=false) {module=SpawnBurst_Instantaneous003}
   &SpawnBurst_Instantaneous(Spawn Group=5, Use Spawn Probability=false, UseLoopCountLimit=false) {module=SpawnBurst_Instantaneous005}
   &SpawnBurst_Instantaneous(Spawn Group=6, Use Spawn Probability=false, UseLoopCountLimit=false) {module=SpawnBurst_Instantaneous006}
   &SpawnBurst_Instantaneous(Spawn Group=7, Use Spawn Probability=false, UseLoopCountLimit=false) {module=SpawnBurst_Instantaneous007}
  ]
  particleSpawnScript [
   /SetVariables(Particles.VisibilityTag=&SpawnGroup()) {isSetParameters=true, module=SetVariables_#3}
   &InitializeParticle(
    Color Mode=::NewEnumerator0
    Lifetime=$Emitter.CurrentLoopDuration
    Position Offset=&MergeHelper_VectorSwitch(
     0=<15 0 0>
     1=<15 0 0>
     2=<16.9 0 0>
     3=<16.9 0 0>
     4=<67 0 0>
     5=<17.3 0 0>
     6=<17.3 0 0>
     7=<17.3 0 0>
     Index=&SpawnGroup()
     Num=::NewEnumerator6
    )
    UsePositionOffset=true
    Write Material Random=true
   )
   /SetVariables(Particles.Initial Light Soft Color:LinearColor=<rgba 1 0.621 0.1255 0.28>, Particles.Initial Light Soft Size:Vector3f=<0.85 0.7 0.85>) {isSetParameters=true}
   /SetVariables(
    Particles.Initial Light Wide Soft Color:LinearColor=<rgba 2 0.9325 0.118 0.9>
    Particles.Initial Light Wide Soft Size:Vector3f=<0.85 0.7 0.85>
   ) {isSetParameters=true}
   /SetVariables(Particles.Initial Light Front Color:LinearColor=<rgba 1 0.6509 0.3804 0.06>, Particles.Initial Light Front Size:Vector3f=<1 1.75 1>) {isSetParameters=true}
   /SetVariables(
    Particles.Initial Light Wide Front Color:LinearColor=<rgba 1 0.5724 0.3263 0.06>
    Particles.Initial Light Wide Front Size:Vector3f=<1 1.75 1>
   ) {isSetParameters=true}
   /SetVariables(Particles.Initial Backward Cone Color:LinearColor=<rgba 1 0.4556 0 0.08>, Particles.Initial Backward Cone Size:Vector3f=<1.1 0.25 1.1>) {isSetParameters=true}
   /SetVariables(
    Particles.Initial Glow Color:LinearColor=&FN_Select_Color_from_Time_Of_Day(Morning=<rgba 1 0.6064 0.1108 0.8>, Day=<rgba 1.1 0.667 0.1219 1>, Evening=<rgba 1 0.6064 0.1108 0.8>, Night=<rgba 1 0.6064 0.1108 0.8>)
    Particles.Initial Glow Size:Vector2f=&Vector2DFromFloat(Value=30)
   ) {isSetParameters=true}
   /SetVariables(
    Particles.Initial Ring Color:LinearColor=&FN_Select_Color_from_Time_Of_Day(Morning=<rgba 1 0.574 0.2065 1>, Day=<rgba 1.4 0.8036 0.289 1>, Evening=<rgba 1 0.574 0.2065 1>, Night=<rgba 1 0.574 0.2065 1>)
    Particles.Initial Ring Size:Vector2f=&Vector2DFromFloat(Value=30)
   ) {isSetParameters=true}
   &InitialMeshOrientation(
    Mesh Axis To Orient=&MergeHelper_VectorSwitch(
     0=<1 0 0>
     1=<1 0 0>
     2=<1 0 0>
     3=<1 0 0>
     4=<0 0 0>
     5=<1 0 0>
     6=<0 0 0>
     7=<0 0 0>
     Index=&SpawnGroup()
     Num=::NewEnumerator6
    )
    Mesh Orientation Mode=::NewEnumerator2
    Orientation Vector=&MergeHelper_VectorSwitch(
     0=<0 1 0>
     1=<0 1 0>
     2=<0 1 0>
     3=<0 1 0>
     4=<0 0 0>
     5=<0 -1 0>
     6=<0 0 0>
     7=<0 0 0>
     Index=&SpawnGroup()
     Num=::NewEnumerator6
    )
    Use Rotation Vector=false
   )
   &CameraOffset(Camera Offset Amount=&MergeHelper_FloatSwitch(0=0, 1=0, 2=0, 3=0, 4=0, 5=1, 6=2, 7=1, Index=&SpawnGroup(), Num=::NewEnumerator6))
  ]
  particleUpdateScript [
   &ParticleState()
   /SetVariables(
    Particles.Scale Light Soft Color:LinearColor=&MakeLinearColorFromVectorAndFloat(
     "Vector (RGB)"=&MakeVectorFromLinearColorRGB(Color=/Niagara/DynamicInputs/Add/Add_LinearColor(
       A:LinearColor=$Particles.Initial.Initial Light Soft Color
       B:LinearColor=/Niagara/DynamicInputs/LinearColor/MakeLinearColorFromHSV(
        HSV From:NiagaraInt32=1
        HSV:Vector3f=?
        Hue:NiagaraFloat=0
        Saturation:NiagaraFloat=0
        Value:NiagaraFloat=&FloatFromCurve(FloatCurve=DataInterface() {propertyValues {Curve {keys [
             {interpMode=RCIM_Cubic, tangentMode=RCTM_User, time=0.007, value=0.7}
             {arriveTangent=-127.0625, interpMode=RCIM_Cubic, leaveTangent=-127.0625, tangentMode=RCTM_User, time=0.02, value=0}
            ]}}})
        Alpha:NiagaraFloat=0
       )
      ))
     "Float (Alpha)"=&MakeFloatFromVector(
      Vector=&VectorFromFloat(Value=&FloatFromCurve(FloatCurve=DataInterface() {propertyValues {Curve {keys [<at 0 1.7>,<at 0.02 1>,<at 0.98 0.9903>,<at 1 0>]}}}))
      Channel=::NewEnumerator0
     )
    )
    Particles.Scale Light Soft Size:Vector3f=&Multiply_Vector(
     A=$Particles.Initial.Initial Light Soft Size
     B=&VectorFromCurve(VectorCurve=DataInterface() {propertyValues {
        XCurve {keys [<at 0 1>]}
        YCurve {keys [
          {arriveTangent=47.6036, interpMode=RCIM_Cubic, leaveTangent=47.6036, tangentMode=RCTM_User, time=0, value=0.5}
          {arriveTangent=37.5, interpMode=RCIM_Cubic, leaveTangent=37.5, time=0.01, value=1.3}
          {
           arriveTangent=-30.1978
           arriveTangentWeight=0.0172
           interpMode=RCIM_Cubic
           leaveTangent=-30.1978
           leaveTangentWeight=0.3242
           tangentMode=RCTM_User
           time=0.016
           value=1.1
          }
          {interpMode=RCIM_Cubic, leaveTangent=-4, time=0.03, value=1}
         ]}
        ZCurve {keys [<at 0 1>]}
       }})
    )
   ) {isSetParameters=true}
   /SetVariables(
    Particles.Scale Light Wide Soft Color:LinearColor=&MakeLinearColorFromVectorAndFloat(
     "Vector (RGB)"=&MakeVectorFromLinearColorRGB(Color=&Add_LinearColor(
       A:LinearColor=$Particles.Initial.Initial Light Wide Soft Color
       B:LinearColor=&MakeLinearColorFromHSV(
        HSV From:NiagaraInt32=1
        HSV:Vector3f=?
        Hue:NiagaraFloat=0
        Saturation:NiagaraFloat=0
        Value:NiagaraFloat=&FloatFromCurve(FloatCurve=DataInterface() {propertyValues {Curve {keys [
             {interpMode=RCIM_Cubic, tangentMode=RCTM_User, time=0.007, value=0.7}
             {arriveTangent=-127.0625, interpMode=RCIM_Cubic, leaveTangent=-127.0625, tangentMode=RCTM_User, time=0.02, value=0}
            ]}}})
        Alpha:NiagaraFloat=0
       )
      ))
     "Float (Alpha)"=&MakeFloatFromVector(
      Vector=&VectorFromFloat(Value=&FloatFromCurve(FloatCurve=DataInterface() {propertyValues {Curve {keys [<at 0 1.7>,<at 0.02 1>,<at 0.98 0.9903>,<at 1 0>]}}}))
      Channel=::NewEnumerator0
     )
    )
    Particles.Scale Light Wide Soft Size:Vector3f=&Multiply_Vector(
     A=$Particles.Initial.Initial Light Wide Soft Size
     B=&VectorFromCurve(VectorCurve=DataInterface() {propertyValues {
        XCurve {keys [<at 0 1>]}
        YCurve {keys [
          {arriveTangent=47.6036, interpMode=RCIM_Cubic, leaveTangent=47.6036, tangentMode=RCTM_User, time=0, value=0.5}
          {arriveTangent=37.5, interpMode=RCIM_Cubic, leaveTangent=37.5, time=0.01, value=1.3}
          {
           arriveTangent=-30.1978
           arriveTangentWeight=0.0172
           interpMode=RCIM_Cubic
           leaveTangent=-30.1978
           leaveTangentWeight=0.3242
           tangentMode=RCTM_User
           time=0.016
           value=1.1
          }
          {interpMode=RCIM_Cubic, leaveTangent=-4, time=0.03, value=1}
         ]}
        ZCurve {keys [<at 0 1>]}
       }})
    )
   ) {isSetParameters=true}
   /SetVariables(
    Particles.Scale Light Front Color:LinearColor=&MakeLinearColorFromVectorAndFloat(
     "Vector (RGB)"=&MakeVectorFromLinearColorRGB(Color=&Add_LinearColor(
       A:LinearColor=$Particles.Initial.Initial Light Front Color
       B:LinearColor=&MakeLinearColorFromHSV(
        HSV From:NiagaraInt32=1
        HSV:Vector3f=?
        Hue:NiagaraFloat=0
        Saturation:NiagaraFloat=0
        Value:NiagaraFloat=&FloatFromCurve(FloatCurve=DataInterface() {propertyValues {Curve {keys [
             {interpMode=RCIM_Cubic, tangentMode=RCTM_User, time=0.007, value=0.7}
             {arriveTangent=-127.0625, interpMode=RCIM_Cubic, leaveTangent=-127.0625, tangentMode=RCTM_User, time=0.02, value=0}
            ]}}})
        Alpha:NiagaraFloat=0
       )
      ))
     "Float (Alpha)"=&Multiply_Float(
      A=&FloatFromCurve(FloatCurve=DataInterface() {propertyValues {Curve {keys [<at 0 1.7>,<at 0.02 1>,<at 0.98 0.9903>,<at 1 0>]}}}, Scale Curve=0.05)
      B=&RemapRange(
       Input Value To Remap=&Abs_Float(Float:NiagaraFloat=&DotProduct(
         Vector A=&ReturnCameraVector(Camera Vectors=::NewEnumerator4, Coordinate Space=::Simulation, Origin Position=$Particles.Position)
         Vector B=&ReturnMeshOrientationAxis(Mesh Orientation Axis=::YAxis, Mesh Orientation Quat=$Particles.MeshOrientation)
        ))
       Input Min=0.62
       Input Max=0.9
       Output Min=0
       Output Max=0.75
      )
     )
    )
    Particles.Scale Light Front Size:Vector3f=&Multiply_Vector(
     A=$Particles.Initial.Initial Light Front Size
     B=&VectorFromCurve(VectorCurve=DataInterface() {propertyValues {
        XCurve {keys [<at 0 1>]}
        YCurve {keys [
          {arriveTangent=47.6036, interpMode=RCIM_Cubic, leaveTangent=47.6036, tangentMode=RCTM_User, time=0, value=0.5}
          {arriveTangent=37.5, interpMode=RCIM_Cubic, leaveTangent=37.5, time=0.01, value=1.3}
          {
           arriveTangent=-30.1978
           arriveTangentWeight=0.0172
           interpMode=RCIM_Cubic
           leaveTangent=-30.1978
           leaveTangentWeight=0.3242
           tangentMode=RCTM_User
           time=0.016
           value=1.1
          }
          {interpMode=RCIM_Cubic, leaveTangent=-4, time=0.03, value=1}
         ]}
        ZCurve {keys [<at 0 1>]}
       }})
    )
   ) {isSetParameters=true}
   /SetVariables(
    Particles.Scale Light Wide Front Color:LinearColor=&MakeLinearColorFromVectorAndFloat(
     "Vector (RGB)"=&MakeVectorFromLinearColorRGB(Color=&Add_LinearColor(
       A:LinearColor=$Particles.Initial.Initial Light Wide Front Color
       B:LinearColor=&MakeLinearColorFromHSV(
        HSV From:NiagaraInt32=1
        HSV:Vector3f=?
        Hue:NiagaraFloat=0
        Saturation:NiagaraFloat=0
        Value:NiagaraFloat=&FloatFromCurve(FloatCurve=DataInterface() {propertyValues {Curve {keys [
             {interpMode=RCIM_Cubic, tangentMode=RCTM_User, time=0.007, value=0.7}
             {arriveTangent=-127.0625, interpMode=RCIM_Cubic, leaveTangent=-127.0625, tangentMode=RCTM_User, time=0.02, value=0}
            ]}}})
        Alpha:NiagaraFloat=0
       )
      ))
     "Float (Alpha)"=&Multiply_Float(
      A=&FloatFromCurve(FloatCurve=DataInterface() {propertyValues {Curve {keys [<at 0 1.7>,<at 0.02 1>,<at 0.98 0.9903>,<at 1 0>]}}}, Scale Curve=0.05)
      B=&RemapRange(
       Input Value To Remap=&Abs_Float(Float:NiagaraFloat=&DotProduct(
         Vector A=&ReturnCameraVector(Camera Vectors=::NewEnumerator4, Coordinate Space=::Simulation, Origin Position=$Particles.Position)
         Vector B=&ReturnMeshOrientationAxis(Mesh Orientation Axis=::YAxis, Mesh Orientation Quat=$Particles.MeshOrientation)
        ))
       Input Min=0.62
       Input Max=0.9
       Output Min=0
       Output Max=0.75
      )
     )
    )
    Particles.Scale Light Wide Front Size:Vector3f=&Multiply_Vector(
     A=$Particles.Initial.Initial Light Wide Front Size
     B=&VectorFromCurve(VectorCurve=DataInterface() {propertyValues {
        XCurve {keys [<at 0 1>]}
        YCurve {keys [
          {arriveTangent=47.6036, interpMode=RCIM_Cubic, leaveTangent=47.6036, tangentMode=RCTM_User, time=0, value=0.5}
          {arriveTangent=37.5, interpMode=RCIM_Cubic, leaveTangent=37.5, time=0.01, value=1.3}
          {
           arriveTangent=-30.1978
           arriveTangentWeight=0.0172
           interpMode=RCIM_Cubic
           leaveTangent=-30.1978
           leaveTangentWeight=0.3242
           tangentMode=RCTM_User
           time=0.016
           value=1.1
          }
          {interpMode=RCIM_Cubic, leaveTangent=-4, time=0.03, value=1}
         ]}
        ZCurve {keys [<at 0 1>]}
       }})
    )
   ) {isSetParameters=true}
   /SetVariables(Particles.Scale Backward Cone Color:LinearColor=&MakeLinearColorFromVectorAndFloat(
     "Vector (RGB)"=&MakeVectorFromLinearColorRGB(Color=$Particles.Initial.Initial Backward Cone Color)
     "Float (Alpha)"=&Multiply_Float(
      A=&FloatFromCurve(
       FloatCurve=DataInterface() {propertyValues {Curve {keys [
           {arriveTangent=1.9063, interpMode=RCIM_Cubic, leaveTangent=1.9063, tangentMode=RCTM_User, time=0.0063, value=0.4157}
           {arriveTangent=0.7524, interpMode=RCIM_Cubic, leaveTangent=0.7524, time=0.35, value=0.9}
           {arriveTangent=-1.0212, interpMode=RCIM_Cubic, leaveTangent=-1.0212, time=0.65, value=0.9}
           {arriveTangent=-3.3789, interpMode=RCIM_Cubic, leaveTangent=-3.3789, tangentMode=RCTM_User, time=1, value=0.2362}
          ]}}}
       Scale Curve=0.5
      )
      B=&FloatFromCurve(
       FloatCurve=DataInterface() {propertyValues {Curve {keys [
           {arriveTangent=0.319, interpMode=RCIM_Cubic, time=0.0316, value=0.4795}
           {arriveTangent=-1.2322, interpMode=RCIM_Cubic, leaveTangent=-1.2322, time=0.1865, value=0.6298}
           {arriveTangent=-0.6628, interpMode=RCIM_Cubic, leaveTangent=-0.6628, time=0.3592, value=0.0758}
           {arriveTangent=0.342, interpMode=RCIM_Cubic, leaveTangent=0.342, time=0.6149, value=0.3458}
           {arriveTangent=-0.7465, interpMode=RCIM_Cubic, leaveTangent=-0.7465, time=0.731, value=0.2029}
           {arriveTangent=-0.7546, interpMode=RCIM_Cubic, leaveTangent=-0.7546, time=0.8255, value=0.1886}
           {arriveTangent=0.759, interpMode=RCIM_Cubic, leaveTangent=0.759, tangentMode=RCTM_User, time=0.9817, value=0.0138}
           {arriveTangent=-24.8817, interpMode=RCIM_Cubic, leaveTangent=-24.8817, tangentMode=RCTM_User, time=0.9984, value=0.0138}
          ]}}}
       Scale Curve=0.25
      )
     )
    )) {isSetParameters=true}
   /SetVariables(Particles.Scale Glow Color:LinearColor=&MakeLinearColorFromVectorAndFloat(
     "Vector (RGB)"=&MakeVectorFromLinearColorRGB(Color=$Particles.Initial.Initial Glow Color)
     "Float (Alpha)"=&Multiply_Float(
      A=&FloatFromCurve(FloatCurve=DataInterface() {propertyValues {Curve {keys [
           {arriveTangent=1.9063, interpMode=RCIM_Cubic, leaveTangent=1.9063, tangentMode=RCTM_User, time=0.0063, value=0.4157}
           {arriveTangent=0.7524, interpMode=RCIM_Cubic, leaveTangent=0.7524, time=0.35, value=0.9}
           {arriveTangent=-1.0212, interpMode=RCIM_Cubic, leaveTangent=-1.0212, time=0.65, value=0.9}
           {arriveTangent=-3.3789, interpMode=RCIM_Cubic, leaveTangent=-3.3789, tangentMode=RCTM_User, time=1, value=0.2362}
          ]}}})
      B=&FloatFromCurve(FloatCurve=DataInterface() {propertyValues {Curve {keys [
           {arriveTangent=0.2579, interpMode=RCIM_Cubic, time=0.0192, value=0.53}
           {arriveTangent=-1.1189, interpMode=RCIM_Cubic, leaveTangent=-1.1189, time=0.1887, value=0.0487}
           {arriveTangent=0.1565, interpMode=RCIM_Cubic, leaveTangent=0.1565, time=0.3511, value=0.1587}
           {arriveTangent=0.1511, interpMode=RCIM_Cubic, leaveTangent=0.1511, time=0.5704, value=0.1084}
           {arriveTangent=-0.221, interpMode=RCIM_Cubic, leaveTangent=-0.221, time=0.6729, value=0.2073}
           {arriveTangent=-0.1468, interpMode=RCIM_Cubic, leaveTangent=-0.1468, tangentMode=RCTM_User, time=0.9984, value=0.0138}
          ]}}})
     )
    )) {isSetParameters=true}
   /SetVariables(Particles.Scale Ring Color:LinearColor=&MakeLinearColorFromVectorAndFloat(
     "Vector (RGB)"=&MakeVectorFromLinearColorRGB(Color=$Particles.Initial Ring Color)
     "Float (Alpha)"=&Multiply_Float(
      A=&FloatFromCurve(FloatCurve=DataInterface() {propertyValues {Curve {keys [
           {arriveTangent=1.9063, interpMode=RCIM_Cubic, leaveTangent=1.9063, tangentMode=RCTM_User, time=0.0063, value=0.4157}
           {arriveTangent=0.7524, interpMode=RCIM_Cubic, leaveTangent=0.7524, time=0.35, value=0.9}
           {arriveTangent=-1.0212, interpMode=RCIM_Cubic, leaveTangent=-1.0212, time=0.65, value=0.9}
           {arriveTangent=-3.3789, interpMode=RCIM_Cubic, leaveTangent=-3.3789, tangentMode=RCTM_User, time=1, value=0.2362}
          ]}}})
      B=&FloatFromCurve(
       FloatCurve=DataInterface() {propertyValues {Curve {keys [
           {arriveTangent=0.2045, interpMode=RCIM_Cubic, time=0.0082, value=0.2754}
           {arriveTangent=-0.1192, interpMode=RCIM_Cubic, leaveTangent=-0.1192, time=0.1887, value=0.0487}
           {arriveTangent=0.0865, interpMode=RCIM_Cubic, leaveTangent=0.0865, time=0.5695, value=0.2086}
           {arriveTangent=-0.4542, interpMode=RCIM_Cubic, leaveTangent=-0.4542, time=0.8031, value=0.1018}
           {arriveTangent=-0.1468, interpMode=RCIM_Cubic, leaveTangent=-0.1468, tangentMode=RCTM_User, time=0.9984, value=0.0138}
          ]}}}
       Scale Curve=0.6
      )
     )
    )) {isSetParameters=true}
  ]
  rendererClasses [&NiagaraMeshRendererProperties, &NiagaraSpriteRendererProperties]
  renderers [
   {rendererClass=&NiagaraMeshRendererProperties, rendererIndex=0}
   {rendererClass=&NiagaraMeshRendererProperties, rendererIndex=1}
   {rendererClass=&NiagaraMeshRendererProperties, rendererIndex=2}
   {rendererClass=&NiagaraMeshRendererProperties, rendererIndex=3}
   {rendererClass=&NiagaraMeshRendererProperties, rendererIndex=4}
   {rendererClass=&NiagaraSpriteRendererProperties, rendererIndex=5}
   {rendererClass=&NiagaraSpriteRendererProperties, rendererIndex=6}
  ]
 }
 Light_Meshes {
  bEnabled=false
  emitterUpdateScript [
   &EmitterState(Life Cycle Mode=::NewEnumerator1, Loop Behavior=::NewEnumerator1, Loop Duration=6.616, UseLoopDelay=false)
   &SpawnBurst_Instantaneous(Use Spawn Probability=false, UseLoopCountLimit=false)
   &SpawnBurst_Instantaneous(Spawn Group=1, Use Spawn Probability=false, UseLoopCountLimit=false) {module=SpawnBurst_Instantaneous001}
  ]
  particleSpawnScript [
   /SetVariables(Particles.VisibilityTag=&SpawnGroup()) {isSetParameters=true, module=SetVariables_#3}
   &InitializeParticle(
    Color=&MergeHelper_ColorSwitch(0=<rgba 2 1.2421 0.251 0.4>, 1=<rgba 2 0.9325 0.118 0.9>, 2=?, 3=?, Index=&SpawnGroup(), Num=::NewEnumerator0)
    Lifetime=$Emitter.CurrentLoopDuration
    Mesh Scale=<0.85 0.7 0.85>
    Mesh Scale Mode=::NewEnumerator4
    Position Offset=<15 0 0>
    UsePositionOffset=true
    Write Material Random=true
   )
   &InitialMeshOrientation(Mesh Orientation Mode=::NewEnumerator2, Orientation Vector=<0 1 0>, Use Rotation Vector=false)
  ]
  particleUpdateScript [
   &ParticleState()
   &ScaleMeshSize(Scale Factor=&VectorFromCurve(VectorCurve=DataInterface() {propertyValues {
       XCurve {keys [<at 0 1>]}
       YCurve {keys [
         {arriveTangent=47.6036, interpMode=RCIM_Cubic, leaveTangent=47.6036, tangentMode=RCTM_User, time=0, value=0.5}
         {arriveTangent=37.5, interpMode=RCIM_Cubic, leaveTangent=37.5, time=0.01, value=1.3}
         {
          arriveTangent=-30.1978
          arriveTangentWeight=0.0172
          interpMode=RCIM_Cubic
          leaveTangent=-30.1978
          leaveTangentWeight=0.3242
          tangentMode=RCTM_User
          time=0.016
          value=1.1
         }
         {interpMode=RCIM_Cubic, leaveTangent=-4, time=0.03, value=1}
        ]}
       ZCurve {keys [<at 0 1>]}
      }}))
   &ScaleColor(
    Scale RGB=&VectorFromFloat(Value=&FloatFromCurve(FloatCurve=DataInterface() {propertyValues {Curve {keys [<at 0 1.7>,<at 0.02 1>,<at 0.98 0.9903>,<at 1 0>]}}}))
    ScaleA=true
    ScaleRGB=true
   )
  ]
  rendererClasses [&NiagaraMeshRendererProperties]
  renderers [
   {rendererClass=&NiagaraMeshRendererProperties, rendererIndex=0}
   {rendererClass=&NiagaraMeshRendererProperties, rendererIndex=1}
  ]
 }
 Light_Meshes001 {
  bEnabled=false
  emitterUpdateScript [
   &EmitterState(Life Cycle Mode=::NewEnumerator1, Loop Behavior=::NewEnumerator1, Loop Duration=6.616, UseLoopDelay=false)
   &SpawnBurst_Instantaneous(Use Spawn Probability=false, UseLoopCountLimit=false)
   &SpawnBurst_Instantaneous(Spawn Group=1, Use Spawn Probability=false, UseLoopCountLimit=false) {module=SpawnBurst_Instantaneous001}
  ]
  particleSpawnScript [
   /SetVariables(Particles.VisibilityTag=&SpawnGroup()) {isSetParameters=true}
   &InitializeParticle(
    Color=&MergeHelper_ColorSwitch(0=<rgba 1 0.6509 0.3804 0.06>, 1=<rgba 1 0.5724 0.3263 0.06>, 2=?, 3=?, Index=$Emitter.SpawnGroup, Num=::NewEnumerator0)
    Lifetime=$Emitter.CurrentLoopDuration
    Mesh Scale=<1 1.75 1>
    Mesh Scale Mode=::NewEnumerator4
    Position Offset=<16.9 0 0>
    UsePositionOffset=true
    Write Material Random=true
   )
   &InitialMeshOrientation(Mesh Orientation Mode=::NewEnumerator2, Orientation Vector=<0 1 0>, Use Rotation Vector=false)
  ]
  particleUpdateScript [
   &ParticleState()
   &ScaleMeshSize(Scale Factor=&VectorFromCurve(VectorCurve=DataInterface() {propertyValues {
       XCurve {keys [<at 0 1>]}
       YCurve {keys [
         {arriveTangent=47.6036, interpMode=RCIM_Cubic, leaveTangent=47.6036, tangentMode=RCTM_User, time=0, value=0.5}
         {arriveTangent=37.5, interpMode=RCIM_Cubic, leaveTangent=37.5, time=0.01, value=1.3}
         {
          arriveTangent=-30.1978
          arriveTangentWeight=0.0172
          interpMode=RCIM_Cubic
          leaveTangent=-30.1978
          leaveTangentWeight=0.3242
          tangentMode=RCTM_User
          time=0.016
          value=1.1
         }
         {interpMode=RCIM_Cubic, leaveTangent=-4, time=0.03, value=1}
        ]}
       ZCurve {keys [<at 0 1>]}
      }}))
   &ScaleColor(
    Scale Alpha=&Multiply_Float(
     A=&FloatFromCurve(FloatCurve=DataInterface() {propertyValues {Curve {keys [<at 0 1.7>,<at 0.02 1>,<at 0.98 0.9903>,<at 1 0>]}}})
     B=&RemapRange(
      Input Value To Remap=&Abs_Float(Float:NiagaraFloat=&DotProduct(
        Vector A=&ReturnCameraVector(Camera Vectors=::NewEnumerator4, Coordinate Space=::Simulation, Origin Position=$Particles.Position)
        Vector B=&ReturnMeshOrientationAxis(Mesh Orientation Axis=::YAxis, Mesh Orientation Quat=$Particles.MeshOrientation)
       ))
      Input Min=0.62
      Input Max=0.9
      Output Min=0
      Output Max=0.75
     )
    )
    ScaleA=true
    ScaleRGB=true
   )
  ]
  rendererClasses [&NiagaraMeshRendererProperties]
  renderers [
   {rendererClass=&NiagaraMeshRendererProperties, rendererIndex=0}
   {rendererClass=&NiagaraMeshRendererProperties, rendererIndex=1}
  ]
 }
 Rainbow_Flare {
  emitterUpdateScript [
   &EmitterState(Life Cycle Mode=::NewEnumerator1, Loop Behavior=::NewEnumerator1, Loop Duration=6.616, UseLoopDelay=false)
   &SpawnBurst_Instantaneous(Use Spawn Probability=false, UseLoopCountLimit=false)
  ]
  particleSpawnScript [
   &InitializeParticle(
    Color Mode=::NewEnumerator0
    Lifetime=6.616
    Position Offset=<67 0 0>
    Sprite Size Mode=::NewEnumerator3
    Uniform Sprite Size=50
    UsePositionOffset=true
    Write Material Random=true
   )
   &SpriteFacingAndAlignment(WriteAlignment=false, WriteFacing=true)
  ]
  particleUpdateScript [
   &ParticleState()
   &ScaleColor(
    Scale Alpha=&Multiply_Float(
     A=&Multiply_Float(
      A=&FloatFromCurve(FloatCurve=DataInterface() {propertyValues {Curve {keys [
           {arriveTangent=18.9034, interpMode=RCIM_Cubic, leaveTangent=18.9034, tangentMode=RCTM_User, time=0, value=0}
           {interpMode=RCIM_Cubic, tangentMode=RCTM_User, time=0.14, value=1}
           {interpMode=RCIM_Cubic, tangentMode=RCTM_User, time=0.86, value=1}
           {arriveTangent=-19.4854, interpMode=RCIM_Cubic, leaveTangent=-19.4854, tangentMode=RCTM_User, time=1, value=0}
          ]}}})
      B=&RemapRange(
       Input Value To Remap=&Abs_Float(Float:NiagaraFloat=&DotProduct(
         Vector A=&ReturnCameraVector(Camera Vectors=::NewEnumerator4, Coordinate Space=::Simulation, Origin Position=$Particles.Position)
         Vector B=$Particles.SpriteFacing
        ))
       Input Min=0.2
       Input Max=0.9
       Output Min=0
       Output Max=1
      )
     )
     B=&Waveform(
      "Waveform [1]"=::NewEnumerator0
      "[1] Frequency"=&RandomRangeFloat(Minimum=0.35, Maximum=0.45)
      "[1] Amplitude Min/Max"=<0.5 0.88>
     )
    )
    ScaleA=true
    ScaleRGB=true
   )
   &ScaleColor(
    Scale Alpha=&RemapRange(
     Input Value To Remap=&Abs_Float(Float:NiagaraFloat=&DotProduct(
       Vector A=&ReturnCameraVector(Camera Vectors=::NewEnumerator4, Coordinate Space=::Simulation, Origin Position=$Particles.Position)
       Vector B=$Particles.SpriteFacing
      ))
     Input Min=0.2
     Input Max=0.9
     Output Min=0
     Output Max=1
    )
    ScaleA=true
    ScaleRGB=true
   ) {enabled=false, module=ScaleColor001}
  ]
  rendererClasses [&NiagaraSpriteRendererProperties]
  renderers [{rendererClass=&NiagaraSpriteRendererProperties}]
 }
 Tertiary_Details {
  bEnabled=false
  emitterUpdateScript [
   &EmitterState(Life Cycle Mode=::NewEnumerator1, Loop Behavior=::NewEnumerator1, Loop Duration=6.616, UseLoopDelay=false)
   &SpawnBurst_Instantaneous(Use Spawn Probability=false, UseLoopCountLimit=false)
   &SpawnBurst_Instantaneous(Spawn Group=1, Use Spawn Probability=false, UseLoopCountLimit=false) {module=SpawnBurst_Instantaneous001}
  ]
  particleSpawnScript [
   &InitializeParticle(
    Color=<rgba 1 0.4556 0 0.08>
    Lifetime=6.616
    Mesh Scale=<1.1 0.25 1.1>
    Mesh Scale Mode=::NewEnumerator4
    Position Offset=<17.3 0 0>
    Sprite Size Mode=::NewEnumerator3
    Uniform Sprite Size=30
    UsePositionOffset=true
    Write Material Random=true
   )
   &InitialMeshOrientation(Mesh Orientation Mode=::NewEnumerator2, Orientation Vector=<0 -1 0>, Use Rotation Vector=false)
   &CameraOffset(Camera Offset Amount=&MaskFloatBySpawnGroup(Masked Float=1, Passthrough Float=3))
   /SetVariables(
    Particles.Linear Color=<rgba 1 0.574 0.2065 0.2>
    Particles.Vector 2D:Vector2f=&Vector2DFromFloat(Value=20)
   ) {isSetParameters=true}
  ]
  particleUpdateScript [
   &ParticleState()
   &UpdateMeshOrientation(Rotation Rate=-0.05, Rotation Vector=<0 -1 0>)
   &ScaleColor(
    Scale Alpha=&Multiply_Float(
     A=&FloatFromCurve(FloatCurve=DataInterface() {propertyValues {Curve {keys [
          {arriveTangent=1.9063, interpMode=RCIM_Cubic, leaveTangent=1.9063, tangentMode=RCTM_User, time=0.0063, value=0.4157}
          {arriveTangent=0.7524, interpMode=RCIM_Cubic, leaveTangent=0.7524, time=0.35, value=0.9}
          {arriveTangent=-1.0212, interpMode=RCIM_Cubic, leaveTangent=-1.0212, time=0.65, value=0.9}
          {arriveTangent=-3.3789, interpMode=RCIM_Cubic, leaveTangent=-3.3789, tangentMode=RCTM_User, time=1, value=0.2362}
         ]}}})
     B=&FloatFromCurve(
      FloatCurve=DataInterface() {propertyValues {Curve {keys [
          {arriveTangent=0.319, interpMode=RCIM_Cubic, time=0.0316, value=0.4795}
          {arriveTangent=-1.2322, interpMode=RCIM_Cubic, leaveTangent=-1.2322, time=0.1865, value=0.6298}
          {arriveTangent=-0.6628, interpMode=RCIM_Cubic, leaveTangent=-0.6628, time=0.3592, value=0.0758}
          {arriveTangent=0.342, interpMode=RCIM_Cubic, leaveTangent=0.342, time=0.6149, value=0.3458}
          {arriveTangent=-0.7465, interpMode=RCIM_Cubic, leaveTangent=-0.7465, time=0.731, value=0.2029}
          {arriveTangent=-0.7546, interpMode=RCIM_Cubic, leaveTangent=-0.7546, time=0.8255, value=0.1886}
          {arriveTangent=0.759, interpMode=RCIM_Cubic, leaveTangent=0.759, tangentMode=RCTM_User, time=0.9817, value=0.0138}
          {arriveTangent=-24.8817, interpMode=RCIM_Cubic, leaveTangent=-24.8817, tangentMode=RCTM_User, time=0.9984, value=0.0138}
         ]}}}
      Scale Curve=0.6
     )
    )
    ScaleA=true
    ScaleRGB=true
   )
   /SetVariables(Particles.Update Glow Color:LinearColor=&MakeLinearColorFromVectorAndFloat(
     "Vector (RGB)"=&MakeVectorFromLinearColorRGB(Color=<rgba 1 0.6064 0.1108 0.09>)
     "Float (Alpha)"=&Multiply_Float(
      A=&FloatFromCurve(FloatCurve=DataInterface() {propertyValues {Curve {keys [
           {arriveTangent=1.9063, interpMode=RCIM_Cubic, leaveTangent=1.9063, tangentMode=RCTM_User, time=0.0063, value=0.4157}
           {arriveTangent=0.7524, interpMode=RCIM_Cubic, leaveTangent=0.7524, time=0.35, value=0.9}
           {arriveTangent=-1.0212, interpMode=RCIM_Cubic, leaveTangent=-1.0212, time=0.65, value=0.9}
           {arriveTangent=-3.3789, interpMode=RCIM_Cubic, leaveTangent=-3.3789, tangentMode=RCTM_User, time=1, value=0.2362}
          ]}}})
      B=&FloatFromCurve(
       FloatCurve=DataInterface() {propertyValues {Curve {keys [
           {arriveTangent=0.2579, interpMode=RCIM_Cubic, time=0.0192, value=0.53}
           {arriveTangent=-1.1189, interpMode=RCIM_Cubic, leaveTangent=-1.1189, time=0.1887, value=0.0487}
           {arriveTangent=0.1565, interpMode=RCIM_Cubic, leaveTangent=0.1565, time=0.3511, value=0.1587}
           {arriveTangent=0.1511, interpMode=RCIM_Cubic, leaveTangent=0.1511, time=0.5704, value=0.1084}
           {arriveTangent=-0.221, interpMode=RCIM_Cubic, leaveTangent=-0.221, time=0.6729, value=0.2073}
           {arriveTangent=-0.1468, interpMode=RCIM_Cubic, leaveTangent=-0.1468, tangentMode=RCTM_User, time=0.9984, value=0.0138}
          ]}}}
       Scale Curve=0.6
      )
     )
    )) {isSetParameters=true}
   /SetVariables(Particles.Update Ring Color:LinearColor=&MakeLinearColorFromVectorAndFloat(
     "Vector (RGB)"=&MakeVectorFromLinearColorRGB(Color=<rgba 1 0.574 0.2065 0.4>)
     "Float (Alpha)"=&Multiply_Float(
      A=&FloatFromCurve(FloatCurve=DataInterface() {propertyValues {Curve {keys [
           {arriveTangent=1.9063, interpMode=RCIM_Cubic, leaveTangent=1.9063, tangentMode=RCTM_User, time=0.0063, value=0.4157}
           {arriveTangent=0.7524, interpMode=RCIM_Cubic, leaveTangent=0.7524, time=0.35, value=0.9}
           {arriveTangent=-1.0212, interpMode=RCIM_Cubic, leaveTangent=-1.0212, time=0.65, value=0.9}
           {arriveTangent=-3.3789, interpMode=RCIM_Cubic, leaveTangent=-3.3789, tangentMode=RCTM_User, time=1, value=0.2362}
          ]}}})
      B=&FloatFromCurve(
       FloatCurve=DataInterface() {propertyValues {Curve {keys [
           {arriveTangent=0.2045, interpMode=RCIM_Cubic, time=0.0082, value=0.2754}
           {arriveTangent=-0.1192, interpMode=RCIM_Cubic, leaveTangent=-0.1192, time=0.1887, value=0.0487}
           {arriveTangent=0.0865, interpMode=RCIM_Cubic, leaveTangent=0.0865, time=0.5695, value=0.2086}
           {arriveTangent=-0.4542, interpMode=RCIM_Cubic, leaveTangent=-0.4542, time=0.8031, value=0.1018}
           {arriveTangent=-0.1468, interpMode=RCIM_Cubic, leaveTangent=-0.1468, tangentMode=RCTM_User, time=0.9984, value=0.0138}
          ]}}}
       Scale Curve=0.6
      )
     )
    )) {isSetParameters=true}
  ]
  rendererClasses [&NiagaraMeshRendererProperties, &NiagaraSpriteRendererProperties]
  renderers [
   {rendererClass=&NiagaraMeshRendererProperties, rendererIndex=0}
   {rendererClass=&NiagaraSpriteRendererProperties, rendererIndex=1}
   {rendererClass=&NiagaraSpriteRendererProperties, rendererIndex=2}
  ]
 }
}
summary {
 systemName=NS_Emote_ClumsyChew
 emitters [
  {emitterName=Faint_Glow, rendererClasses []}
  {emitterName=Light, rendererClasses []}
  {emitterName=Glow, rendererClasses []}
  {emitterName=Flare, rendererClasses []}
  {emitterName=Rainbow_Flare, rendererClasses [&NiagaraSpriteRendererProperties]}
  {emitterName=Dust, rendererClasses [&NiagaraSpriteRendererProperties]}
  {emitterName=Hot_Ring, rendererClasses []}
  {
   bEnabled=false
   emitterName=Tertiary_Details
   rendererClasses [&NiagaraMeshRendererProperties, &NiagaraSpriteRendererProperties]
  }
  {bEnabled=false, emitterName=Light_Meshes, rendererClasses [&NiagaraMeshRendererProperties]}
  {bEnabled=false, emitterName=Light_Meshes001, rendererClasses [&NiagaraMeshRendererProperties]}
  {emitterName=Light_Cones, rendererClasses [&NiagaraMeshRendererProperties, &NiagaraSpriteRendererProperties]}
 ]
 userVariables []
}
